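# Sound cue for quiet one-shot daytime bird calls in the Knights forest biome, using bird attenuation and the outdoor ambient one-shot class.
# A quality node sends its first two slots to a large wave pool and its last two to a seven-player pool excluded from branch culling.
# Each pool randomly picks one of two modulated branches, which randomizes pitch and volume over a randomly chosen call.
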SoundCue /Game/Sounds/Asteria/Biomes/Knights/Spots/Spot_Knights_Forest_Day_Bird_Cue
root=QualityLevel_0
nodes [
 {id=Modulator_2, children [Random_2], properties {pitchMin=0.3, volumeMax=0.6, volumeMin=0.3}}
 {id=Modulator_3, children [Random_5], properties {pitchMin=0.3, volumeMax=0.6, volumeMin=0.3}}
 {id=Modulator_4, children [Random_8], properties {pitchMin=0.3, volumeMax=0.6, volumeMin=0.3}}
 {id=Modulator_5, children [Random_9], properties {pitchMin=0.3, volumeMax=0.6, volumeMin=0.3}}
 {id=QualityLevel_0, children [Random_7,Random_7,Random_10,Random_10]}
 {id=Random_10, children [Modulator_4,Modulator_5], properties {bShouldExcludeFromBranchCulling=true, weights [1,1]}}
 {
  id=Random_2
  children [
   WavePlayer_74
   WavePlayer_75
   WavePlayer_76
   WavePlayer_77
   WavePlayer_78
   WavePlayer_79
   WavePlayer_80
   WavePlayer_81
   WavePlayer_82
   WavePlayer_83
   WavePlayer_84
   WavePlayer_85
   WavePlayer_86
   WavePlayer_87
   WavePlayer_88
   WavePlayer_89
   WavePlayer_90
   WavePlayer_91
   WavePlayer_92
   WavePlayer_93
   WavePlayer_94
   WavePlayer_95
   WavePlayer_96
   WavePlayer_97
   WavePlayer_98
   WavePlayer_99
   WavePlayer_100
   WavePlayer_101
   WavePlayer_102
   WavePlayer_103
   WavePlayer_104
   WavePlayer_105
  ]
  properties {weights [1,1,1,1,1,1,1,1,1,1,1,1,1,1,1,1,1,1,1,1,1,1,1,1,1,1,1,1,1,1,1,1]}
 }
 {
  id=Random_5
  children [
   WavePlayer_175
   WavePlayer_176
   WavePlayer_177
   WavePlayer_178
   WavePlayer_179
   WavePlayer_180
   WavePlayer_181
   WavePlayer_182
   WavePlayer_183
   WavePlayer_184
   WavePlayer_185
   WavePlayer_186
   WavePlayer_187
   WavePlayer_188
   WavePlayer_189
   WavePlayer_190
   WavePlayer_191
   WavePlayer_192
   WavePlayer_193
   WavePlayer_194
   WavePlayer_195
   WavePlayer_196
   WavePlayer_197
   WavePlayer_198
   WavePlayer_199
   WavePlayer_200
   WavePlayer_201
   WavePlayer_202
   WavePlayer_203
   WavePlayer_204
  ]
  properties {weights [1,1,1,1,1,1,1,1,1,1,1,1,1,1,1,1,1,1,1,1,1,1,1,1,1,1,1,1,1,1]}
 }
 {id=Random_7, children [Modulator_2,Modulator_3], properties {weights [1,1]}}
 {
  id=Random_8
  children [WavePlayer_205, WavePlayer_206, WavePlayer_207, WavePlayer_208]
  properties {bShouldExcludeFromBranchCulling=true, weights [1,1,1,1]}
 }
 {id=Random_9, children [WavePlayer_209, WavePlayer_210, WavePlayer_211], properties {bShouldExcludeFromBranchCulling=true, weights [1,1,1]}}
 {id=WavePlayer_100, properties {soundWaveAssetPtr=&/Spot_Knights_Forest_Day_Bird_Call_26}}
 {id=WavePlayer_101, properties {soundWaveAssetPtr=&/Spot_Knights_Forest_Day_Bird_Call_27}}
 {id=WavePlayer_102, properties {soundWaveAssetPtr=&/Spot_Knights_Forest_Day_Bird_Call_28}}
 {id=WavePlayer_103, properties {soundWaveAssetPtr=&/Spot_Knights_Forest_Day_Bird_Call_29}}
 {id=WavePlayer_104, properties {soundWaveAssetPtr=&/Spot_Knights_Forest_Day_Bird_Call_30}}
 {id=WavePlayer_105, properties {soundWaveAssetPtr=&/Spot_Knights_Forest_Day_Bird_Call_31}}
 {id=WavePlayer_175, properties {soundWaveAssetPtr=&/Spot_Knights_Forest_Day_Bird_Call_66}}
 {id=WavePlayer_176, properties {soundWaveAssetPtr=&/Spot_Knights_Forest_Day_Bird_Call_33}}
 {id=WavePlayer_177, properties {soundWaveAssetPtr=&/Spot_Knights_Forest_Day_Bird_Call_34}}
 {id=WavePlayer_178, properties {soundWaveAssetPtr=&/Spot_Knights_Forest_Day_Bird_Call_35}}
 {id=WavePlayer_179, properties {soundWaveAssetPtr=&/Spot_Knights_Forest_Day_Bird_Call_36}}
 {id=WavePlayer_180, properties {soundWaveAssetPtr=&/Spot_Knights_Forest_Day_Bird_Call_37}}
 {id=WavePlayer_181, properties {soundWaveAssetPtr=&/Spot_Knights_Forest_Day_Bird_Call_38}}
 {id=WavePlayer_182, properties {soundWaveAssetPtr=&/Spot_Knights_Forest_Day_Bird_Call_40}}
 {id=WavePlayer_183, properties {soundWaveAssetPtr=&/Spot_Knights_Forest_Day_Bird_Call_41}}
 {id=WavePlayer_184, properties {soundWaveAssetPtr=&/Spot_Knights_Forest_Day_Bird_Call_42}}
 {id=WavePlayer_185, properties {soundWaveAssetPtr=&/Spot_Knights_Forest_Day_Bird_Call_43}}
 {id=WavePlayer_186, properties {soundWaveAssetPtr=&/Spot_Knights_Forest_Day_Bird_Call_44}}
 {id=WavePlayer_187, properties {soundWaveAssetPtr=&/Spot_Knights_Forest_Day_Bird_Call_45}}
 {id=WavePlayer_188, properties {soundWaveAssetPtr=&/Spot_Knights_Forest_Day_Bird_Call_46}}
 {id=WavePlayer_189, properties {soundWaveAssetPtr=&/Spot_Knights_Forest_Day_Bird_Call_47}}
 {id=WavePlayer_190, properties {soundWaveAssetPtr=&/Spot_Knights_Forest_Day_Bird_Call_48}}
 {id=WavePlayer_191, properties {soundWaveAssetPtr=&/Spot_Knights_Forest_Day_Bird_Call_49}}
 {id=WavePlayer_192, properties {soundWaveAssetPtr=&/Spot_Knights_Forest_Day_Bird_Call_50}}
 {id=WavePlayer_193, properties {soundWaveAssetPtr=&/Spot_Knights_Forest_Day_Bird_Call_51}}
 {id=WavePlayer_194, properties {soundWaveAssetPtr=&/Spot_Knights_Forest_Day_Bird_Call_52}}
 {id=WavePlayer_195, properties {soundWaveAssetPtr=&/Spot_Knights_Forest_Day_Bird_Call_53}}
 {id=WavePlayer_196, properties {soundWaveAssetPtr=&/Spot_Knights_Forest_Day_Bird_Call_54}}
 {id=WavePlayer_197, properties {soundWaveAssetPtr=&/Spot_Knights_Forest_Day_Bird_Call_55}}
 {id=WavePlayer_198, properties {soundWaveAssetPtr=&/Spot_Knights_Forest_Day_Bird_Call_56}}
 {id=WavePlayer_199, properties {soundWaveAssetPtr=&/Spot_Knights_Forest_Day_Bird_Call_58}}
 {id=WavePlayer_200, properties {soundWaveAssetPtr=&/Spot_Knights_Forest_Day_Bird_Call_59}}
 {id=WavePlayer_201, properties {soundWaveAssetPtr=&/Spot_Knights_Forest_Day_Bird_Call_61}}
 {id=WavePlayer_202, properties {soundWaveAssetPtr=&/Spot_Knights_Forest_Day_Bird_Call_62}}
 {id=WavePlayer_203, properties {soundWaveAssetPtr=&/Spot_Knights_Forest_Day_Bird_Call_63}}
 {id=WavePlayer_204, properties {soundWaveAssetPtr=&/Spot_Knights_Forest_Day_Bird_Call_65}}
 {id=WavePlayer_205, properties {soundWaveAssetPtr=&/Spot_Knights_Forest_Day_Bird_Call_16}}
 {id=WavePlayer_206, properties {soundWaveAssetPtr=&/Spot_Knights_Forest_Day_Bird_Call_12}}
 {id=WavePlayer_207, properties {soundWaveAssetPtr=&/Spot_Knights_Forest_Day_Bird_Call_11}}
 {id=WavePlayer_208, properties {soundWaveAssetPtr=&/Spot_Knights_Forest_Day_Bird_Call_01}}
 {id=WavePlayer_209, properties {soundWaveAssetPtr=&/Spot_Knights_Forest_Day_Bird_Call_39}}
 {id=WavePlayer_210, properties {soundWaveAssetPtr=&Spot_Knights_Forest_Day_Bird_Call_38}}
 {id=WavePlayer_211, properties {soundWaveAssetPtr=&Spot_Knights_Forest_Day_Bird_Call_36}}
 {id=WavePlayer_74, properties {soundWaveAssetPtr=&Spot_Knights_Forest_Day_Bird_Call_66}}
 {id=WavePlayer_75, properties {soundWaveAssetPtr=&Spot_Knights_Forest_Day_Bird_Call_01}}
 {id=WavePlayer_76, properties {soundWaveAssetPtr=&/Spot_Knights_Forest_Day_Bird_Call_02}}
 {id=WavePlayer_77, properties {soundWaveAssetPtr=&/Spot_Knights_Forest_Day_Bird_Call_03}}
 {id=WavePlayer_78, properties {soundWaveAssetPtr=&/Spot_Knights_Forest_Day_Bird_Call_04}}
 {id=WavePlayer_79, properties {soundWaveAssetPtr=&/Spot_Knights_Forest_Day_Bird_Call_05}}
 {id=WavePlayer_80, properties {soundWaveAssetPtr=&/Spot_Knights_Forest_Day_Bird_Call_06}}
 {id=WavePlayer_81, properties {soundWaveAssetPtr=&/Spot_Knights_Forest_Day_Bird_Call_07}}
 {id=WavePlayer_82, properties {soundWaveAssetPtr=&/Spot_Knights_Forest_Day_Bird_Call_08}}
 {id=WavePlayer_83, properties {soundWaveAssetPtr=&/Spot_Knights_Forest_Day_Bird_Call_09}}
 {id=WavePlayer_84, properties {soundWaveAssetPtr=&/Spot_Knights_Forest_Day_Bird_Call_10}}
 {id=WavePlayer_85, properties {soundWaveAssetPtr=&Spot_Knights_Forest_Day_Bird_Call_11}}
 {id=WavePlayer_86, properties {soundWaveAssetPtr=&Spot_Knights_Forest_Day_Bird_Call_12}}
 {id=WavePlayer_87, properties {soundWaveAssetPtr=&/Spot_Knights_Forest_Day_Bird_Call_13}}
 {id=WavePlayer_88, properties {soundWaveAssetPtr=&/Spot_Knights_Forest_Day_Bird_Call_14}}
 {id=WavePlayer_89, properties {soundWaveAssetPtr=&/Spot_Knights_Forest_Day_Bird_Call_15}}
 {id=WavePlayer_90, properties {soundWaveAssetPtr=&Spot_Knights_Forest_Day_Bird_Call_16}}
 {id=WavePlayer_91, properties {soundWaveAssetPtr=&/Spot_Knights_Forest_Day_Bird_Call_17}}
 {id=WavePlayer_92, properties {soundWaveAssetPtr=&/Spot_Knights_Forest_Day_Bird_Call_18}}
 {id=WavePlayer_93, properties {soundWaveAssetPtr=&/Spot_Knights_Forest_Day_Bird_Call_19}}
 {id=WavePlayer_94, properties {soundWaveAssetPtr=&/Spot_Knights_Forest_Day_Bird_Call_20}}
 {id=WavePlayer_95, properties {soundWaveAssetPtr=&/Spot_Knights_Forest_Day_Bird_Call_21}}
 {id=WavePlayer_96, properties {soundWaveAssetPtr=&/Spot_Knights_Forest_Day_Bird_Call_22}}
 {id=WavePlayer_97, properties {soundWaveAssetPtr=&/Spot_Knights_Forest_Day_Bird_Call_23}}
 {id=WavePlayer_98, properties {soundWaveAssetPtr=&/Spot_Knights_Forest_Day_Bird_Call_24}}
 {id=WavePlayer_99, properties {soundWaveAssetPtr=&/Spot_Knights_Forest_Day_Bird_Call_25}}
]
properties {
 attenuationSettings=&Fort_TileAmbient_OneShot_Bird_Att
 bEnableBusSends=false
 maxDistance=8005
 soundClassObject=&Fort_Ambient_OneShots_Outdoors
 volumeMultiplier=0.05
 preEffectBusSends [{}]
}
