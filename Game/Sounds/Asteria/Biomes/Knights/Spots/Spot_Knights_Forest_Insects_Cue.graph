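# Sound Cue for a Knights forest biome ambient spot, named for insects but playing Woods biome bird one-shots.
# A Random node picks among fourteen equally weighted bird waves borrowed from the Apollo sound set.
# A Modulator above it varies pitch and volume on each play for variety.
# Output uses the outdoor ambient sound class, a shared tile-ambient one-shot attenuation and a reduced cue volume.
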SoundCue /Game/Sounds/Asteria/Biomes/Knights/Spots/Spot_Knights_Forest_Insects_Cue
root=Modulator_0
nodes [
 {id=Modulator_0, children [Random_1], properties {pitchMax=1.2, pitchMin=0.8, volumeMax=1.2, volumeMin=0.8}}
 {
  id=Random_1
  children [
   WavePlayer_6
   WavePlayer_9
   WavePlayer_11
   WavePlayer_12
   WavePlayer_13
   WavePlayer_14
   WavePlayer_24
   WavePlayer_25
   WavePlayer_26
   WavePlayer_27
   WavePlayer_28
   WavePlayer_29
   WavePlayer_30
   WavePlayer_31
  ]
  properties {weights [1,1,1,1,1,1,1,1,1,1,1,1,1,1]}
 }
 {id=WavePlayer_11, properties {soundWaveAssetPtr=/Game/Sounds/Apollo/Biomes/Woods/Biome_Woods_Bird_10}}
 {id=WavePlayer_12, properties {soundWaveAssetPtr=/Game/Sounds/Apollo/Biomes/Woods/Biome_Woods_Bird_11}}
 {id=WavePlayer_13, properties {soundWaveAssetPtr=/Game/Sounds/Apollo/Biomes/Woods/Biome_Woods_Bird_12}}
 {id=WavePlayer_14, properties {soundWaveAssetPtr=/Game/Sounds/Apollo/Biomes/Woods/Biome_Woods_Bird_13}}
 {id=WavePlayer_24, properties {soundWaveAssetPtr=/Game/Sounds/Apollo/Biomes/Woods/Biome_Woods_Bird_14}}
 {id=WavePlayer_25, properties {soundWaveAssetPtr=/Game/Sounds/Apollo/Biomes/Woods/Biome_Woods_Bird_01}}
 {id=WavePlayer_26, properties {soundWaveAssetPtr=/Game/Sounds/Apollo/Biomes/Woods/Biome_Woods_Bird_02}}
 {id=WavePlayer_27, properties {soundWaveAssetPtr=/Game/Sounds/Apollo/Biomes/Woods/Biome_Woods_Bird_03}}
 {id=WavePlayer_28, properties {soundWaveAssetPtr=/Game/Sounds/Apollo/Biomes/Woods/Biome_Woods_Bird_04}}
 {id=WavePlayer_29, properties {soundWaveAssetPtr=/Game/Sounds/Apollo/Biomes/Woods/Biome_Woods_Bird_05}}
 {id=WavePlayer_30, properties {soundWaveAssetPtr=/Game/Sounds/Apollo/Biomes/Woods/Biome_Woods_Bird_06}}
 {id=WavePlayer_31, properties {soundWaveAssetPtr=/Game/Sounds/Apollo/Biomes/Woods/Biome_Woods_Bird_07}}
 {id=WavePlayer_6, properties {soundWaveAssetPtr=/Game/Sounds/Apollo/Biomes/Woods/Biome_Woods_Bird_08}}
 {id=WavePlayer_9, properties {soundWaveAssetPtr=/Game/Sounds/Apollo/Biomes/Woods/Biome_Woods_Bird_09}}
]
properties {
 attenuationSettings=/Game/Sounds/Attenuation/Fort_TileAmbient_OneShot_Att
 maxDistance=8000
 soundClassObject=&Fort_Ambient_Outdoors_Athena
 volumeMultiplier=0.551
 preEffectBusSends [{}]
}
