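# Night bird spot one-shot cue for the Knights forest biome, reusing Apollo Grassland bird recordings.
# A quality-level root routes the first two tiers to six mono waves and the last two to their stereo versions.
# Each set is chosen by an equal-weight Random node, then passed through a Modulator with lowered minimum pitch and volume for wide variation.
# Playback uses bird one-shot attenuation and the ambient one-shot sound class at reduced volume.
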SoundCue /Game/Sounds/Asteria/Biomes/Knights/Spots/Spot_Knights_Forest_NightBird_Cue
root=QualityLevel_0
nodes [
 {id=Modulator_0, children [Random_0], properties {pitchMin=0.3, volumeMin=0.3}}
 {id=Modulator_1, children [Random_2], properties {pitchMin=0.3, volumeMin=0.3}}
 {id=QualityLevel_0, children [Modulator_0,Modulator_0,Modulator_1,Modulator_1]}
 {id=Random_0, children [WavePlayer_1,WavePlayer_2,WavePlayer_3,WavePlayer_4,WavePlayer_5,WavePlayer_6], properties {weights [1,1,1,1,1,1]}}
 {
  id=Random_2
  children [WavePlayer_12, WavePlayer_13, WavePlayer_14, WavePlayer_15, WavePlayer_16, WavePlayer_17]
  properties {weights [1,1,1,1,1,1]}
 }
 {id=WavePlayer_1, properties {soundWaveAssetPtr=/Game/Sounds/Apollo/Biomes/Grassland_01/Biome_Grassland_01_NightBird_01}}
 {id=WavePlayer_12, properties {soundWaveAssetPtr=/Game/Sounds/Apollo/Biomes/Grassland_01/Biome_Grassland_01_NightBird_01_Stereo}}
 {id=WavePlayer_13, properties {soundWaveAssetPtr=/Game/Sounds/Apollo/Biomes/Grassland_01/Biome_Grassland_01_NightBird_02_Stereo}}
 {id=WavePlayer_14, properties {soundWaveAssetPtr=/Game/Sounds/Apollo/Biomes/Grassland_01/Biome_Grassland_01_NightBird_03_Stereo}}
 {id=WavePlayer_15, properties {soundWaveAssetPtr=/Game/Sounds/Apollo/Biomes/Grassland_01/Biome_Grassland_01_NightBird_04_Stereo}}
 {id=WavePlayer_16, properties {soundWaveAssetPtr=/Game/Sounds/Apollo/Biomes/Grassland_01/Biome_Grassland_01_NightBird_05_Stereo}}
 {id=WavePlayer_17, properties {soundWaveAssetPtr=/Game/Sounds/Apollo/Biomes/Grassland_01/Biome_Grassland_01_NightBird_06_Stereo}}
 {id=WavePlayer_2, properties {soundWaveAssetPtr=/Game/Sounds/Apollo/Biomes/Grassland_01/Biome_Grassland_01_NightBird_02}}
 {id=WavePlayer_3, properties {soundWaveAssetPtr=/Game/Sounds/Apollo/Biomes/Grassland_01/Biome_Grassland_01_NightBird_03}}
 {id=WavePlayer_4, properties {soundWaveAssetPtr=/Game/Sounds/Apollo/Biomes/Grassland_01/Biome_Grassland_01_NightBird_04}}
 {id=WavePlayer_5, properties {soundWaveAssetPtr=/Game/Sounds/Apollo/Biomes/Grassland_01/Biome_Grassland_01_NightBird_05}}
 {id=WavePlayer_6, properties {soundWaveAssetPtr=/Game/Sounds/Apollo/Biomes/Grassland_01/Biome_Grassland_01_NightBird_06}}
]
properties {
 attenuationSettings=&Fort_TileAmbient_OneShot_Bird_Att
 bEnableBusSends=false
 maxDistance=8005
 soundClassObject=&Fort_Ambient_OneShots
 volumeMultiplier=0.351
 preEffectBusSends [{}]
}
