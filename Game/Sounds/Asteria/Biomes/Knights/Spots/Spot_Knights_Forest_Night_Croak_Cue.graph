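# Sound cue for a spot ambient croak, named for the Knights biome forest at night, reusing Artemis Reality Forest AlienCroak waves.
# A quality-level root sends its first two slots to a random pick of nine croaks and its last two to a smaller pool of the first four.
# Each pool passes through a modulator varying pitch and volume.
# It plays as an ambient one-shot with reduced volume and wider tile-ambient attenuation.
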SoundCue /Game/Sounds/Asteria/Biomes/Knights/Spots/Spot_Knights_Forest_Night_Croak_Cue
root=QualityLevel_0
nodes [
 {id=Modulator_1, children [Random_2], properties {pitchMax=1.1, volumeMax=1.1, volumeMin=0.9}}
 {id=Modulator_2, children [Random_3], properties {pitchMax=1.1, volumeMax=1.1, volumeMin=0.9}}
 {id=QualityLevel_0, children [Modulator_1,Modulator_1,Modulator_2,Modulator_2]}
 {
  id=Random_2
  children [WavePlayer_0,WavePlayer_1,WavePlayer_2,WavePlayer_3,WavePlayer_4,WavePlayer_5,WavePlayer_6,WavePlayer_8,WavePlayer_9]
  properties {weights [1,1,1,1,1,1,1,1,1]}
 }
 {id=Random_3, children [WavePlayer_10, WavePlayer_11, WavePlayer_12, WavePlayer_13], properties {weights [1,1,1,1]}}
 {id=WavePlayer_0, properties {soundWaveAssetPtr=/Game/Sounds/Artemis/Biomes/Reality_Forest/Spots/AMB_Biome_Reality_Forest_SS_AlienCroak_01}}
 {id=WavePlayer_1, properties {soundWaveAssetPtr=/Game/Sounds/Artemis/Biomes/Reality_Forest/Spots/AMB_Biome_Reality_Forest_SS_AlienCroak_02}}
 {id=WavePlayer_10, properties {soundWaveAssetPtr=&AMB_Biome_Reality_Forest_SS_AlienCroak_01}}
 {id=WavePlayer_11, properties {soundWaveAssetPtr=&AMB_Biome_Reality_Forest_SS_AlienCroak_02}}
 {id=WavePlayer_12, properties {soundWaveAssetPtr=/Game/Sounds/Artemis/Biomes/Reality_Forest/Spots/AMB_Biome_Reality_Forest_SS_AlienCroak_03}}
 {id=WavePlayer_13, properties {soundWaveAssetPtr=/Game/Sounds/Artemis/Biomes/Reality_Forest/Spots/AMB_Biome_Reality_Forest_SS_AlienCroak_04}}
 {id=WavePlayer_2, properties {soundWaveAssetPtr=&AMB_Biome_Reality_Forest_SS_AlienCroak_03}}
 {id=WavePlayer_3, properties {soundWaveAssetPtr=&AMB_Biome_Reality_Forest_SS_AlienCroak_04}}
 {id=WavePlayer_4, properties {soundWaveAssetPtr=/Game/Sounds/Artemis/Biomes/Reality_Forest/Spots/AMB_Biome_Reality_Forest_SS_AlienCroak_05}}
 {id=WavePlayer_5, properties {soundWaveAssetPtr=/Game/Sounds/Artemis/Biomes/Reality_Forest/Spots/AMB_Biome_Reality_Forest_SS_AlienCroak_06}}
 {id=WavePlayer_6, properties {soundWaveAssetPtr=/Game/Sounds/Artemis/Biomes/Reality_Forest/Spots/AMB_Biome_Reality_Forest_SS_AlienCroak_07}}
 {id=WavePlayer_8, properties {soundWaveAssetPtr=/Game/Sounds/Artemis/Biomes/Reality_Forest/Spots/AMB_Biome_Reality_Forest_SS_AlienCroak_08}}
 {id=WavePlayer_9, properties {soundWaveAssetPtr=/Game/Sounds/Artemis/Biomes/Reality_Forest/Spots/AMB_Biome_Reality_Forest_SS_AlienCroak_09}}
]
properties {
 attenuationSettings=/Game/Sounds/Attenuation/Fort_TileAmbient_OneShot_Wider_Att
 bEnableBusSends=false
 maxDistance=8000
 soundClassObject=&Fort_Ambient_OneShots
 volumeMultiplier=0.376
 preEffectBusSends [{}]
}
