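# Sound cue for a one-shot ambient spot of wind rustling foliage in the Knights forest biome, reusing Grassland rustle waves.
# A quality-level node sends its first two slots to an equal-weight random pick of five waves and its last two to a three-wave subset.
# Each branch passes through its own modulator for slight pitch and volume variation.
# Playback uses a wind one-shot attenuation asset, the ambient one-shot sound class and a reduced volume.
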SoundCue /Game/Sounds/Asteria/Biomes/Knights/Spots/Spot_Knights_Forest_Wind_Foliage_Rustle_Cue
root=QualityLevel_0
nodes [
 {id=Modulator_1, children [Random_2], properties {pitchMax=1, pitchMin=0.9, volumeMax=1.1, volumeMin=0.9}}
 {id=Modulator_2, children [Random_4], properties {pitchMax=1, volumeMax=1.1, volumeMin=0.9}}
 {id=QualityLevel_0, children [Modulator_1,Modulator_1,Modulator_2,Modulator_2]}
 {id=Random_2, children [WavePlayer_8, WavePlayer_9, WavePlayer_10, WavePlayer_11, WavePlayer_12], properties {weights [1,1,1,1,1]}}
 {id=Random_4, children [WavePlayer_20, WavePlayer_23, WavePlayer_25], properties {weights [1,1,1]}}
 {
  id=WavePlayer_10
  properties {soundWaveAssetPtr=/Game/Sounds/Artemis/Biomes/Grasslands/Grassland_Spots/Spot_Grassland_Wind_Foliage_Rustle_02}
 }
 {
  id=WavePlayer_11
  properties {soundWaveAssetPtr=/Game/Sounds/Artemis/Biomes/Grasslands/Grassland_Spots/Spot_Grassland_Wind_Foliage_Rustle_03}
 }
 {
  id=WavePlayer_12
  properties {soundWaveAssetPtr=/Game/Sounds/Artemis/Biomes/Grasslands/Grassland_Spots/Spot_Grassland_Wind_Foliage_Rustle_04}
 }
 {
  id=WavePlayer_20
  properties {soundWaveAssetPtr=/Game/Sounds/Artemis/Biomes/Grasslands/Grassland_Spots/Spot_Grassland_Wind_Foliage_Rustle_01}
 }
 {
  id=WavePlayer_23
  properties {soundWaveAssetPtr=&Spot_Grassland_Wind_Foliage_Rustle_03}
 }
 {
  id=WavePlayer_25
  properties {soundWaveAssetPtr=&Spot_Grassland_Wind_Foliage_Rustle_02}
 }
 {id=WavePlayer_8, properties {soundWaveAssetPtr=/Game/Sounds/Artemis/Biomes/Grasslands/Grassland_Spots/Spot_Grassland_Wind_Foliage_Rustle_05}}
 {id=WavePlayer_9, properties {soundWaveAssetPtr=&Spot_Grassland_Wind_Foliage_Rustle_01}}
]
properties {
 attenuationSettings=/Game/Sounds/Attenuation/Fort_TileAmbient_OneShot_Wind_Att
 bEnableBusSends=false
 maxDistance=8000
 soundClassObject=&Fort_Ambient_OneShots
 volumeMultiplier=0.376
 preEffectBusSends [{}]
}
